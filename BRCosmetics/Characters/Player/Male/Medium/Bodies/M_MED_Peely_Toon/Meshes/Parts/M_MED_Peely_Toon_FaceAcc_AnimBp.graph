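# Post-process style animation Blueprint for Toon Peely's face accessory, which follows the attached parent's pose and adds secondary physics.
# Constraints blend the two flap bases between spine_05 and neck_01, then feed a root-space RigidBody solve.
# A chained pelvis-space ADS RigidBody has zero alpha, and targeting floats set in the event graph go unused.
# Each update also sets a smile-mouth material parameter, identically whether emoting or not.
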AnimBlueprint /BRCosmetics/Characters/Player/Male/Medium/Bodies/M_MED_Peely_Toon/Meshes/Parts/M_MED_Peely_Toon_FaceAcc_AnimBp
target_skeleton=&/M_MED_Peely_Toon_FaceAcc_Skeleton
graphs [
 AnimGraph {nodes [
   "Transform (Modify) Bone - Bone: dyn_flapBase_r"(Alpha=1, Translation="0,0,0", Rotation="0,-18,0", Scale="1,1,1") {
    rotationMode=BMM_Additive
    rotationSpace=BCS_BoneSpace
    translationSpace=BCS_BoneSpace
    alphaBoolBlend {blendOption=Linear}
    boneToModify {boneName=dyn_flapBase_r}
    rotation {yaw=-18}
   }
   RigidBody() {
    bClampLinearTranslationLimitToRefPose=true
    bOverrideWorldGravity=true
    overridePhysicsAsset=&/M_MED_Peely_Toon_FaceAcc_Physics
    simulationSpace=BaseBoneSpace
    alphaBoolBlend {blendOption=Linear}
    baseBoneRef {boneName=root}
    componentAppliedLinearAccClamp=<800 800 800>
    componentLinearAccScale=<0.65 0.65 0.65>
    componentLinearVelScale=<0.3 0.3 0.3>
    overrideWorldGravity {z=-980}
   }
   n1 = "Constraint - dyn_flapBase_r (PP)"!(
    ComponentPose=Local To Component(LocalPose=Copy Pose From Mesh() {bCopyCurves=true, bUseAttachedParent=true})
    Alpha=1
    ConstraintWeights_0=0.8
    ConstraintWeights_1:float=0.2
   ) -> (Pose) {
    alphaBoolBlend {blendOption=Linear}
    boneToModify {boneName=dyn_flapBase_r}
    constraintSetup [{targetBone {boneName=spine_05}}, {targetBone {boneName=neck_01}}]
    constraintWeights [1,1]
   }
   n2 = "Constraint - dyn_flap_Base_l (PP)"!(ComponentPose=@n1, Alpha=1, ConstraintWeights_0=0.8, ConstraintWeights_1:float=0.2) -> (Pose) {
    alphaBoolBlend {blendOption=Linear}
    boneToModify {boneName=dyn_flap_Base_l}
    constraintSetup [{targetBone {boneName=spine_05}}, {targetBone {boneName=neck_01}}]
    constraintWeights [1,1]
   }
   n3 = RigidBody(ComponentPose=@n2) {
    bClampLinearTranslationLimitToRefPose=true
    bOverrideWorldGravity=true
    overridePhysicsAsset=&M_MED_Peely_Toon_FaceAcc_Physics
    simulationSpace=BaseBoneSpace
    alphaBoolBlend {blendOption=Linear}
    baseBoneRef {boneName=root}
    componentAppliedLinearAccClamp=<800 800 800>
    componentLinearAccScale=<0.65 0.65 0.65>
    componentLinearVelScale=<0.3 0.3 0.3>
    overrideWorldGravity {z=-980}
   }
   n4 = RigidBody(ComponentPose=@n3) {
    alpha=0
    bClampLinearTranslationLimitToRefPose=true
    bOverrideWorldGravity=true
    overridePhysicsAsset=&/M_MED_Peely_Toon_FaceAcc_ADS_Physics
    simulationSpace=BaseBoneSpace
    alphaBoolBlend {blendOption=Linear}
    baseBoneRef {boneName=pelvis}
    componentAppliedLinearAccClamp=<800 800 800>
    componentLinearAccScale=<0.65 0.65 0.65>
    componentLinearVelScale=<0.3 0.3 0.3>
    overrideWorldGravity {z=-980}
   }
   Output Pose(Result=Component To Local(ComponentPose=@n4)) {layerGroup=None}
   Get isTargeting() -> (isTargeting:double)
   Get isNotTargeting() -> (isNotTargeting:double)
  ]}
 EventGraph {nodes [
   Event BlueprintUpdateAnimation()
   > n1 = Sequence!(execute) -> (then_0, then_1, then_2:Exec)
   > n2 = Branch(Condition=Get bIsTargeting())
   > Set isTargeting(isTargeting:double=1) -> (Output_Get:double)
   Set isNotTargeting(execute=@n2.else, isNotTargeting:double=1) -> (Output_Get:double)
   n3 = Get Dynamic Material Instance(execute=@n1.then_1, SkeletalMeshComponent=GetPartSkeletalMeshComponent())
   > n4 = Branch(Condition=Get bIsPlayingEmote())
   > SetScalarParameterValue(self=@n3, ParameterName=Smile Mouth animation Frame, Value=1)
   SetScalarParameterValue(execute=@n4.else, self=@n3, ParameterName=Smile Mouth animation Frame, Value=1)
   TryGetPawnOwner()
  ]}
]
variables [
 bIsTargeting_0 {type=boolean}
 isNotTargeting {type=number}
 isTargeting {type=number}
 Skeletal Mesh Component {type=SkeletalMeshComponent}
]
